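# Conversation database for the BatterBoi NPC, serving as the top-level entry for talking to this character.
# The entry point leads to one speech node that picks a greeting from contextual lines gated on first-ever meeting, per-match conversation count or a specific outfit, with priority rows for first meeting and outfit.
# That node then branches through links to remote entries for NPC library quests, basic services, a Season 33 quest part and BatterBoi's own dialogue.
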ConversationDatabase /NPCLibrary/NPCs/BatterBoi/NPCLibrary_Conversation_BatterBoi
entry_points [{tag=Conversation.Entry.BatterBoi, destinations [#1]}]
nodes [
 ConversationEntryPointNode_7 {
  id=#1
  class=ConversationEntryPointNode
  outputs [#2]
  properties {entryTag {tagName=Conversation.Entry.BatterBoi}}
 }
 Speech_C_4 {
  id=#2
  class=Speech_C
  outputs [#3, #4, #5, #6]
  properties {
   nodeName=""
   generalConfig {
    defaultMessage="These buns are fresh outta the oven, bud."
    contextualMessages [
     {
      message="Hey there, lil roll."
      contextRequirements [{requirement {class=FortControllerRequirement_IsFirstEverConversationWithNPC}}]
      priority {curve {rowName=Tandem.Speech.Priority.FirstEverInteraction}}
     }
     {
      message="These buns are fresh outta the oven, bud."
      contextRequirements [{requirement {class=FortControllerRequirement_CheckNPCMatchConversationCount, properties {nPCUniqueIdTag {tagName=AISpawnerData.Type.Tandem.BatterBoi}}}}]
      priority {curve {rowName=None}}
     }
     {
      message="Rise and lift, buddy, rise and lift."
      contextRequirements [{requirement {class=FortControllerRequirement_CheckNPCMatchConversationCount, properties {nPCUniqueIdTag {tagName=AISpawnerData.Type.Tandem.BatterBoi}}}}]
      priority {curve {rowName=None}}
     }
     {
      message="You don't gotta be strong to have a strong heart!"
      contextRequirements [{requirement {class=FortControllerRequirement_CheckNPCMatchConversationCount, properties {nPCUniqueIdTag {tagName=AISpawnerData.Type.Tandem.BatterBoi}}}}]
      priority {curve {rowName=None}}
     }
     {
      message="Keep fighting the good fight out there, demon slayer."
      contextRequirements [{requirement {class=FortControllerRequirement_HasCID, properties {allowedCIds [&Character_CubicVice_Dual]}}}]
      priority {curve {rowName=Tandem.Speech.Priority.OutfitReaction}}
     }
    ]
   }
  }
 }
 ConversationLinkNode_3 {id=#6, class=ConversationLinkNode, properties {remoteEntryTag {tagName=Conversation.Entry.BatterBoi.Dialogue}}}
 ConversationLinkNode_2 {
  id=#5
  class=ConversationLinkNode
  properties {remoteEntryTag {tagName=Conversation.Entry.Season33.WF.Quest.04.part.03}}
 }
 ConversationLinkNode_0 {id=#4, class=ConversationLinkNode, properties {remoteEntryTag {tagName=Conversation.Entry.BasicServices}}}
 ConversationLinkNode_1 {id=#3, class=ConversationLinkNode, properties {remoteEntryTag {tagName=Conversation.Entry.NPCLibrary.Quests}}}
]
